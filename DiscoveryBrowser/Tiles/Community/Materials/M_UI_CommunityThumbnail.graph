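# Translucent UI material for a community tile thumbnail in the Discovery Browser, drawing a parameterized image with a rounded-corner shape.
# The ThumbnailImage texture is scaled to fill through MF_UI_UVScaling around a Pivot X/Y point, and its RGB is multiplied by itself.
# A rounded-mask function supplies the shape, and a lerp on rounded vertical UV keeps that rounding to the top half, leaving bottom corners square.
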Material /DiscoveryBrowser/Tiles/Community/Materials/M_UI_CommunityThumbnail
graph {
 properties {
  bAutomaticallySetUsageInEditor=false
  bUsedWithStaticMesh=false
  blendMode=BLEND_Translucent
  materialDomain=MD_UI
  usedShadingModels=Default Lit
 }
 nodes [
  n1 = MaterialFunctionCall() {materialFunction=&GetUserInterfaceUV}
  LinearInterpolate(
   A=MaterialFunctionCall(Corner Radius=Constant() {r=10}, Radius in Pixels=Constant() {r=1}) {materialFunction=&MF_UI_CoreUI_RoundedMask}.Shape Mask
   Alpha=Round(None=ComponentMask(None=@n1.Normalized UV) {g=true})
  )
  n2 = TextureObjectParameter() {group=Thumnail Image, parameterName=ThumbnailImage, texture=/Game/Global/Textures/Default/Blanks/T_Blank_White_ZeroAlpha_sRGBColor}
  n3 = ScalarParameter() {defaultValue=0.5, group=Thumnail Image, parameterName=Pivot X, sortPriority=33}
  n4 = ScalarParameter() {group=Thumnail Image, parameterName=Pivot Y, sortPriority=34}
  n5 = TextureSample(
   UVs=MaterialFunctionCall(
    UV=MaterialFunctionCall() {materialFunction=&GetUserInterfaceUV}.Normalized UV
    TextureSize=TextureProperty(Tex=@n2)
    ScalePivotPoint=AppendVector(A=@n3, B=@n4)
   ) {materialFunction=&MF_UI_UVScaling}.Scaled To Fill
   Tex=@n2
  ) {texture=/MeshModelingToolsetExp/Textures/DefaultTexture}
  Multiply(A=@n5.RGB, B=@n5.RGB)
 ]
}
