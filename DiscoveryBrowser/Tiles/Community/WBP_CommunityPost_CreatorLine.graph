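# Creator line widget for community post tiles in the Discovery Browser, showing who made a post and how long ago.
# An Overlay layers a horizontal row of creator avatar, name, dot divider and time text over a rounded badge backplate.
# ConvertCreatedDateTimeToText turns a DateTime into "time ago" text.
# The event graph is empty, and the file shows nothing that calls that function or fills the placeholder texts.
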WidgetBlueprint /DiscoveryBrowser/Tiles/Community/WBP_CommunityPost_CreatorLine
graphs [
 ConvertCreatedDateTimeToText {nodes [
   n1 = ConvertCreatedDateTimeToText!() -> (then, DateTime:Date Time Structure) {is=title K2Node_FunctionEntry}
   > n2 = ConvertDateTimeToTimeAgoText!(execute, DateTime:Date Time Structure=@n1) -> (then, Text:Text)
   > Return Node!(execute, Text:Text=@n2)
  ]}
 EventGraph {links [], nodes []}
]
variables []
widgets {
 info {inheritedWidgetCount=0, parentClass=&UserWidget, rootWidgetClass=&Overlay, widgetCount=7}
 widgets [
  Overlay_216 {widgetClassPath=&Overlay, widget=Overlay_216}
  Border_Backplate {
   widgetClassPath=&Border
   parent=Overlay_216
   widget=Border_Backplate
   properties {
    background {
     bIsDynamicallyLoaded=false
     drawAs=RoundedBox
     imageType=FullColor
     mirroring=NoMirror
     resourceName=None
     resourceObject=/FortUILibrary/Systems/Badge/MaterialInstances/MI_UIKit_BadgeBackground
     tiling=NoTile
     imageSize=<32 32>
     margin=<ltrb 0 0 0 0>
     outlineSettings {
      bUseBrushTransparency=false
      roundingType=HalfHeightRadius
      width=0
      color {colorUseRule=UseColor_Specified, specifiedColor=<rgba 0 0 0 0>}
      cornerRadii=<0 0 0 0>
     }
     tintColor {colorUseRule=UseColor_Specified, specifiedColor=<rgba 1 1 1 1>}
     uVRegion {bIsValid=false, max=<0 0>, min=<0 0>}
    }
    brushColor=<rgba 1 1 1 0.9>
    padding=<ltrb 0 0 0 0>
   }
   slot {class=OverlaySlot, properties {horizontalAlignment=HAlign_Fill, verticalAlignment=VAlign_Fill}}
  }
  HorizontalBox_Content {
   widgetClassPath=&HorizontalBox
   parent=Overlay_216
   widget=HorizontalBox_Content
   properties {bOverride_Cursor=true}
   slot {class=OverlaySlot, properties {horizontalAlignment=HAlign_Fill, verticalAlignment=VAlign_Fill, padding=<ltrb 2 2 10 2>}}
  }
  CreatorAvatar {
   widgetClassPath=/FortniteCoreUI/UI/Community/WBP_Community_CreatorAvatar
   bIsVariable=true
   parent=HorizontalBox_Content
   widget=CreatorAvatar
   slot {class=HorizontalBoxSlot, properties {padding=<ltrb 0 0 4 0>}}
  }
  Text_Creator {
   widgetClassPath=&CommonTextBlock
   parent=HorizontalBox_Content
   widget=Text_Creator
   properties {
    mobileFontSizeMultiplier=1.4
    text=Creator
    font {
     bForceMonospaced=false
     bMaterialIsStencil=false
     fontMaterial=null
     fontObject=&HeadingNow
     letterSpacing=10
     monospacedWidth=1
     size=11.25
     skewAmount=0
     typefaceFontName=74 Regular
     outlineSettings {
      bApplyOutlineToDropShadows=false
      bMiteredCorners=false
      bSeparateFillAlpha=false
      outlineMaterial=null
      outlineSize=0
      outlineColor=<rgba 0 0 0 1>
     }
    }
    shadowOffset=<0 0>
   }
   slot {class=HorizontalBoxSlot, properties {verticalAlignment=VAlign_Center}}
  }
  Text_Divider {
   widgetClassPath=&CommonTextBlock
   parent=HorizontalBox_Content
   widget=Text_Divider
   properties {
    mobileFontSizeMultiplier=1.4
    text="·"
    colorAndOpacity {colorUseRule=UseColor_Specified, specifiedColor=<rgba 0.5776 0.5776 0.5776 1>}
    font {
     bForceMonospaced=false
     bMaterialIsStencil=false
     fontMaterial=null
     fontObject=&HeadingNow
     letterSpacing=10
     monospacedWidth=1
     size=11.25
     skewAmount=0
     typefaceFontName=74 Regular
     outlineSettings {
      bApplyOutlineToDropShadows=false
      bMiteredCorners=false
      bSeparateFillAlpha=false
      outlineMaterial=null
      outlineSize=0
      outlineColor=<rgba 0 0 0 1>
     }
    }
    shadowOffset=<0 0>
   }
   slot {class=HorizontalBoxSlot, properties {verticalAlignment=VAlign_Center, padding=<ltrb 4 0 4 0>}}
  }
  Text_Time {
   widgetClassPath=&CommonTextBlock
   parent=HorizontalBox_Content
   widget=Text_Time
   properties {
    mobileFontSizeMultiplier=1.4
    text=time ago
    colorAndOpacity {colorUseRule=UseColor_Specified, specifiedColor=<rgba 0.5776 0.5776 0.5776 1>}
    font {
     bForceMonospaced=false
     bMaterialIsStencil=false
     fontMaterial=null
     fontObject=&HeadingNow
     letterSpacing=10
     monospacedWidth=1
     size=11.25
     skewAmount=0
     typefaceFontName=74 Regular
     outlineSettings {
      bApplyOutlineToDropShadows=false
      bMiteredCorners=false
      bSeparateFillAlpha=false
      outlineMaterial=null
      outlineSize=0
      outlineColor=<rgba 0 0 0 1>
     }
    }
    shadowOffset=<0 0>
   }
   slot {class=HorizontalBoxSlot, properties {verticalAlignment=VAlign_Center}}
  }
 ]
}
