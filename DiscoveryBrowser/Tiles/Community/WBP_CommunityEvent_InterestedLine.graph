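# Community event tile widget in the Discovery browser that shows an "interested" count beside an event replies reactions line.
# SetInterestedCount shows the interested row only when Count is above zero, otherwise hiding it with its layout space kept.
# For positive counts it writes the number as shorthand text.
# The replies line is a nested WBP_CommunityPost_ReactionsLine that this graph leaves untouched, and both construct events are empty.
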WidgetBlueprint /DiscoveryBrowser/Tiles/Community/WBP_CommunityEvent_InterestedLine
graphs [
 EventGraph {nodes [
   Event PreConstruct()
   Event Construct()
  ]}
 SetInterestedCount {nodes [
   n1 = SetInterestedCount!() -> (then, Count:int) {is=title K2Node_FunctionEntry}
   n2 = "integer > integer"(A=@n1)
   > SetVisibility(
    self=Get HB_Interested() -> (HB_Interested:Horizontal Box Object Reference)
    InVisibility="To Visibility (Boolean)"!(bIsVisible:bool=@n2, TrueVisibility:ESlateVisibility Enum=SelfHitTestInvisible, FalseVisibility:ESlateVisibility Enum=Hidden) -> (ReturnValue:ESlateVisibility Enum=Visible)
   )
   > Branch(Condition=@n2)
   > "SetText (Text)"(
    self=Get Text_InterestedCount() -> (Text_InterestedCount:Common Text Object Reference)
    InText=ConvertQuantityToShorthandText!(Quantity:int=Get Count() -> (Count:int=0)) -> (Text:Text)
   )
  ]}
]
variables [
 TextPadding_CompactUI {type=Margin, default=<ltrb 8 0 0 0>}
 TextPadding_Default {type=Margin, default=<ltrb 4 0 0 0>}
]
widgets {
 info {inheritedWidgetCount=0, parentClass=&UserWidget, rootWidgetClass=&GridPanel, widgetCount=5}
 widgets [
  GridPanel_0 {widgetClassPath=&GridPanel, widget=GridPanel_0, properties {bOverride_Cursor=true, columnFill [1], rowFill [1]}}
  HB_Interested {
   widgetClassPath=&HorizontalBox
   bIsVariable=true
   parent=GridPanel_0
   widget=HB_Interested
   slot {class=GridSlot, properties {verticalAlignment=VAlign_Center}}
  }
  Text_InterestedCount {
   widgetClassPath=&CommonTextBlock
   bIsVariable=true
   parent=HB_Interested
   widget=Text_InterestedCount
   properties {
    bIsScrollingEnabled=false
    mobileFontSizeMultiplier=1.25
    text="{count}"
    font {
     bForceMonospaced=false
     bMaterialIsStencil=false
     fontMaterial=null
     fontObject=&HeadingNow
     letterSpacing=0
     monospacedWidth=1
     size=12.75
     skewAmount=0
     typefaceFontName=75 Medium
     outlineSettings {
      bApplyOutlineToDropShadows=false
      bMiteredCorners=false
      bSeparateFillAlpha=false
      outlineMaterial=null
      outlineSize=0
      outlineColor=<rgba 0 0 0 1>
     }
    }
    shadowOffset=<0 0>
   }
   slot {class=HorizontalBoxSlot, properties {horizontalAlignment=HAlign_Left, verticalAlignment=VAlign_Center}}
  }
  Text_CommunityGameMode_2 {
   widgetClassPath=&CommonTextBlock
   bIsVariable=true
   parent=HB_Interested
   widget=Text_CommunityGameMode_2
   properties {
    bIsScrollingEnabled=false
    mobileFontSizeMultiplier=1.25
    text=interested
    colorAndOpacity {colorUseRule=UseColor_Specified, specifiedColor=<rgba 0.5776 0.5776 0.5776 1>}
    font {
     bForceMonospaced=false
     bMaterialIsStencil=false
     fontMaterial=null
     fontObject=&HeadingNow
     letterSpacing=0
     monospacedWidth=1
     size=12.75
     skewAmount=0
     typefaceFontName=74 Regular
     outlineSettings {
      bApplyOutlineToDropShadows=false
      bMiteredCorners=false
      bSeparateFillAlpha=false
      outlineMaterial=null
      outlineSize=0
      outlineColor=<rgba 0 0 0 1>
     }
    }
    shadowOffset=<0 0>
   }
   slot {class=HorizontalBoxSlot, properties {horizontalAlignment=HAlign_Left, verticalAlignment=VAlign_Center, padding=<ltrb 4 0 0 0>}}
  }
  Count_EventReplies {
   widgetClassPath=/FortniteCoreUI/UI/Community/WBP_CommunityPost_ReactionsLine
   bIsVariable=true
   parent=GridPanel_0
   widget=Count_EventReplies
   properties {emoji="💬", material=/FortUILibrary/Textures/MaterialInstances/MI_UIKit_Icon_ChannelWhsiper, in Color and Opacity=<rgba 0.5191 0.7428 1 1>}
   slot {class=GridSlot, properties {column=1, horizontalAlignment=HAlign_Right, padding=<ltrb 12 0 0 0>}}
  }
 ]
}
