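# Activatable modal widget in the Discovery browser that shows a community post over a click-catching scrim, with Report and Back buttons in a bottom action bar.
# Activation is forwarded to the embedded post content widget, which supplies the focus target and triggers a focus refresh when its query succeeds.
# Back, Report (after calling the content's Report) and the content's CTA all deactivate the modal.
# PreConstruct adjusts modal height by aspect ratio and prompt padding for compact UI.
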WidgetBlueprint /DiscoveryBrowser/Tiles/Community/WBP_CommunityPost_Modal
graphs [
 BP_GetDesiredFocusTarget {nodes [
   Get Desired Focus Target()
   > Return Node!(
    execute
    ReturnValue:Widget Object Reference=GetFocusTarget!(self:WBP Community Post Content Object Reference=Get WBP_CommunityPost_Content() -> (WBP_CommunityPost_Content:WBP Community Post Content Object Reference)) -> (TargetWidget:Widget Object Reference)
   )
  ]}
 EventGraph {nodes [
   "On Clicked (Button_Back)"!() -> (OutputDelegate, then, Button:Common Button Base Object Reference) {is=title K2Node_ComponentBoundEvent}
   > DeactivateWidget()
   "On Clicked (Button_Report)"!() -> (OutputDelegate, then, Button:Common Button Base Object Reference) {is=title K2Node_ComponentBoundEvent}
   > Report(self:WBP Community Post Content Object Reference=Get WBP_CommunityPost_Content())
   > DeactivateWidget()
   "On CTAButton Clicked (WBP_CommunityPost_Content)"() {is=title K2Node_ComponentBoundEvent}
   > DeactivateWidget()
   "On Success Finished Querying (WBP_CommunityPost_Content)"() {is=title K2Node_ComponentBoundEvent}
   > RequestRefreshFocus(self:Common Activatable Widget Object Reference)
   Event On Activated()
   > OnActivated(self:WBP Community Post Content Object Reference=Get WBP_CommunityPost_Content())
   Event PreConstruct()
   > SetCompactUIStyle() {is=title K2Node_CallFunction}
  ]}
 SetCompactUIStyle {nodes [
   SetCompactUIStyle() {is=title K2Node_FunctionEntry}
   > Branch(Condition=IsUsingCompactUI())
   > SetHeightOverride(
    self=Get SB_Modal() -> (SB_Modal:Size Box Object Reference)
    InHeightOverride=SelectFloat(
     A=1150
     B=950
     bPickA=CompareToScreenAspectRatio!(Comparison:EAspectRatioComparison Enum=LessThanOrEqualTo, AspectRatio:EAspectRatio Enum=Ratio_23_16) -> (ReturnValue:bool)
    )
   )
   > SetPadding!(
    execute
    self:Overlay Slot Object Reference=SlotAsOverlaySlot(Widget=Get SafeZone_Prompts() -> (SafeZone_Prompts:Safe Zone Object Reference))
    InPadding_Left:float=48
    InPadding_Top:float=0
    InPadding_Right:float=48
    InPadding_Bottom:float=24
   ) -> (then)
  ]}
]
variables [
 CommunityHighlightId {type=string, default=""}
 IsEvent {type=boolean}
]
widgets {
 info {
  inheritedWidgetCount=0
  namedSlotCount=1
  parentClass=&CommonActivatableWidget
  rootWidgetClass=&Overlay
  widgetCount=9
 }
 widgets [
  Overlay_0 {widgetClassPath=&Overlay, widget=Overlay_0}
  WBP_UIKit_Scrim {
   widgetClassPath=/FortUILibrary/Systems/Dialog/WBP_UIKit_Scrim
   bIsVariable=true
   parent=Overlay_0
   widget=WBP_UIKit_Scrim
   properties {useClickCatcher=true}
   slot {class=OverlaySlot, properties {horizontalAlignment=HAlign_Fill, verticalAlignment=VAlign_Fill}}
  }
  SB_Modal {
   widgetClassPath=&SizeBox
   bIsVariable=true
   parent=Overlay_0
   widget=SB_Modal
   properties {bOverride_HeightOverride=true, bOverride_WidthOverride=true, heightOverride=920, widthOverride=1648}
   slot {class=OverlaySlot, properties {horizontalAlignment=HAlign_Center, verticalAlignment=VAlign_Center}}
  }
  WBP_CommunityPost_Content {
   widgetClassPath=/FortniteCoreUI/UI/Community/WBP_CommunityPost_Content
   bIsVariable=true
   parent=SB_Modal
   widget=WBP_CommunityPost_Content
   slot {class=SizeBoxSlot}
  }
  SafeZone_Prompts {
   widgetClassPath=&SafeZone
   bIsVariable=true
   parent=Overlay_0
   widget=SafeZone_Prompts
   properties {padTop=false}
   slot {class=OverlaySlot, properties {horizontalAlignment=HAlign_Right, verticalAlignment=VAlign_Bottom, padding=<ltrb 80 0 80 32>}}
  }
  WBP_ActionBarContainer {
   widgetClassPath=/Game/UI/Frontend/BottomBar/WBP_ActionBarContainer
   bIsVariable=true
   parent=SafeZone_Prompts
   widget=WBP_ActionBarContainer
   slot {class=SafeZoneSlot}
  }
  HorizontalBox_1 {
   widgetClassPath=&HorizontalBox
   namedSlotHost=WBP_ActionBarContainer
   widget=HorizontalBox_1
   properties {flowDirectionPreference=Culture}
  }
  Button_Report {
   widgetClassPath=/FortUILibrary/UIKit/Buttons/WBP_UIKit_Button_Quiet
   bIsVariable=true
   parent=HorizontalBox_1
   widget=Button_Report
   properties {bIsFocusable=false, text=Report Post, triggeringInputAction {dataTable=&InputActionDataTable, rowName=ReportPost}}
   slot {class=HorizontalBoxSlot, properties {horizontalAlignment=HAlign_Right, verticalAlignment=VAlign_Bottom}}
  }
  Button_Back {
   widgetClassPath=&WBP_UIKit_Button_Quiet
   bIsVariable=true
   parent=HorizontalBox_1
   widget=Button_Back
   properties {bIsFocusable=false, text=Back, triggeringInputAction {dataTable=&InputActionDataTable, rowName=GenericBack}}
   slot {class=HorizontalBoxSlot, properties {horizontalAlignment=HAlign_Right, verticalAlignment=VAlign_Bottom, padding=<ltrb 24 0 0 0>}}
  }
 ]
}
