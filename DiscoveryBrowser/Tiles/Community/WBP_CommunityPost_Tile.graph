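# Discovery browser tile for a community highlight post or creator event, showing thumbnail, title, island, creator, reactions or event timing.
# Conversion functions derive status-based text, badges, visibility and size from view model data; runtime options gate the event layout.
# Tile display and clicks send impressions; clicks route through the activity selector, then push a post modal.
# Context menu actions open replies, creator page or island details.
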WidgetBlueprint /DiscoveryBrowser/Tiles/Community/WBP_CommunityPost_Tile
graphs [
 CF_CreatedDateTimeToText {nodes [
   n1 = CF_CreatedDateTimeToText!() -> (then, DateTime:Date Time Structure) {is=title K2Node_FunctionEntry}
   > n2 = ConvertDateTimeToTimeAgoText!(execute, DateTime:Date Time Structure=@n1) -> (then, Text:Text)
   > Return Node!(execute, Text:Text=@n2)
  ]}
 CF_DateTimeToEndedText {nodes [
   n1 = CF_DateTimeToEndedText!() -> (then, EndTime:Date Time Structure) {is=title K2Node_FunctionEntry}
   n2 = DateTime - DateTime!(A:Date Time Structure=UTC Now!() -> (ReturnValue:Date Time Structure), B:Date Time Structure=@n1) -> (ReturnValue:Timespan Structure)
   n3 = "Max (Integer)"(
    A=Round(A=Get Total Days!(A:Timespan Structure=@n2) -> (ReturnValue:double) {is=title K2Node_CallFunction}) {is=title K2Node_CallFunction}
    B=0
   )
   n4 = Format Text!(Format="LOCTEXT(\"ENDED {Time} {Time}|plural(one=DAY,other=DAYS) AGO\")", Time:int=@n3) -> (Result)
   n5 = "integer > integer"(A=@n3)
   n6 = "Max (Integer)"(
    A=Floor(A=Get Total Hours!(A:Timespan Structure=@n2) -> (ReturnValue:double) {is=title K2Node_CallFunction}) {is=title K2Node_CallFunction}
    B=0
   )
   n7 = Get Minutes!(A:Timespan Structure=@n2) -> (ReturnValue:int=0) {is=title K2Node_CallFunction}
   n8 = SelectText(
    A=Format Text(Format="LOCTEXT(\"ENDED {Time} {Time}|plural(one=MINUTE,other=MINUTES) AGO\")", Time="Max (Integer)"(A=@n7, B=0))
    B=Format Text(Format="LOCTEXT(\"ENDED {Time} {Time}|plural(one=HOUR,other=HOURS) AGO\")", Time=@n6)
    bPickA="integer <= integer"(A=@n6)
   )
   n9 = Get Total Days(A=@n2)
   n10 = "Max (Integer)"(A=int / int(A=Floor(A=@n9), B=30), B=0)
   > Return Node(Text=SelectText(
     A=Format Text(Format="LOCTEXT(\"ENDED {Time} {Time}|plural(one=MONTH,other=MONTHS) AGO\")", Time=@n10)
     B=SelectText(A=@n4, B=@n8, bPickA=@n5)
     bPickA="integer > integer"(A=@n10)
    ))
  ]}
 CF_DateTimeToLiveText {nodes [
   n1 = CF_DateTimeToLiveText!() -> (then, EndTime:Date Time Structure) {is=title K2Node_FunctionEntry}
   n2 = DateTime - DateTime(A=@n1, B=UTC Now())
   n3 = Round(A=Get Total Days(A=@n2))
   n4 = Format Text!(Format="LOCTEXT(\"ONGOING • ENDS IN {Days} {Days}|plural(one=DAY,other=DAYS)\")", Days:int=@n3) -> (Result)
   n5 = "integer > integer"(A=@n3)
   n6 = Floor(A=Get Total Hours(A=@n2))
   n7 = SelectText(
    A=Format Text!(Format="LOCTEXT(\"ONGOING • ENDS IN {Time} {Time}|plural(one=MINUTE,other=MINUTES)\")", Time:int=Get Minutes(A=@n2) -> (ReturnValue=0)) -> (Result)
    B=Format Text(Format="LOCTEXT(\"ONGOING • ENDS IN {Time} {Time}|plural(one=HOUR,other=HOURS)\")", Time=@n6)
    bPickA="integer <= integer"(A=@n6)
   )
   n8 = int / int(A=Floor(A=Get Total Days(A=@n2)), B=30)
   > Return Node(Text=SelectText(
     A=Format Text!(Format="LOCTEXT(\"ONGOING • ENDS IN {Months} {Months}|plural(one=MONTH,other=MONTHS)\")", Months:int=@n8) -> (Result)
     B=SelectText(A=@n4, B=@n7, bPickA=@n5)
     bPickA="integer > integer"(A=@n8)
    ))
  ]}
 CF_GetDateSideColor {nodes [
   n1 = CF_GetDateSideColor!() -> (then, IsInterested:bool=~, EventStatus:ECommunityEventStatus Enum=Unknown) {is=title K2Node_FunctionEntry}
   > Return Node!(
    execute
    Color:Linear Color Structure=GetDateSideColor!(
     self:WBP Community Date Object Reference=Get StartDate() -> (StartDate:WBP Community Date Object Reference)
     IsInterested:bool=@n1
     EventStatus:ECommunityEventStatus Enum=@n1.EventStatus
    ) -> (ReturnValue:Linear Color Structure)
   )
  ]}
 CF_HighlightText {nodes [
   CF_HighlightText!() -> (then, IsInterested:bool=~, EventStatus:ECommunityEventStatus Enum=Unknown) {is=title K2Node_FunctionEntry}
   n1 = Get EventStatus() -> (EventStatus:ECommunityEventStatus Enum=Unknown)
   > n2 = Branch(Condition=OR Boolean(A="Equal (Enum)"(A:ECommunityEventStatus Enum=@n1, B:ECommunityEventStatus Enum=Ended), B="Equal (Enum)"(A=@n1, B=Canceled)))
   Return Node!(
    execute=[@n2.then, @n3.then]
    Text:Text=Select!(
     Unknown:Text
     Upcoming:Text
     Live:Text="LOCTEXT(\"Live\")"
     Ended:Text
     Canceled:Text="LOCTEXT(\"Canceled\")"
     Index:ECommunityEventStatus Enum=Get EventStatus() -> (EventStatus=Unknown)
    ) -> (ReturnValue:Text)
   )
   n3 = Branch(execute=@n2.else, Condition=NOT Boolean(A=Get IsInterested() -> (IsInterested:bool)))
   Return Node(execute=@n3.else, Text="LOCTEXT(\"Interested\")")
  ]}
 CF_HighlightType {nodes [
   CF_HighlightType!() -> (then, IsInterested:bool=~, EventStatus:ECommunityEventStatus Enum=Unknown) {is=title K2Node_FunctionEntry}
   n1 = Get EventStatus() -> (EventStatus=Unknown)
   n2 = Get IsInterested()
   > Return Node!(
    execute
    HighlightType:E_UI_HighlightType Enum=Select!(
     Option 0:E_UI_HighlightType Enum=Select(Option 0=NewEnumerator8, Option 1=NewEnumerator0, Index=@n2)
     Option 1:E_UI_HighlightType Enum=NewEnumerator3
     Index:bool="Equal (Enum)"(A=@n1, B=Canceled)
    ) -> (ReturnValue:E_UI_HighlightType Enum)
   )
  ]}
 CF_HighlightVisibility {nodes [
   n1 = CF_HighlightVisibility!() -> (then, IsInterested:bool=~, EventStatus:ECommunityEventStatus Enum=Unknown) {is=title K2Node_FunctionEntry}
   n2 = "Equal (Enum)"(A=@n1.EventStatus, B=Canceled)
   n3 = "Equal (Enum)"(A=@n1.EventStatus, B=Live)
   > Return Node!(
    execute
    Visibility:ESlateVisibility Enum="To Visibility (Boolean)"!(
     bIsVisible:bool=OR Boolean!(A=@n1, B=@n3, C:bool=@n2) -> (ReturnValue)
     TrueVisibility:ESlateVisibility Enum=SelfHitTestInvisible
     FalseVisibility:ESlateVisibility Enum=Collapsed
    ) -> (ReturnValue:ESlateVisibility Enum=Visible)
   )
  ]}
 CF_ParamsToTileSize {nodes [
   n1 = CF_ParamsToTileSize!() -> (then, ColumnSize:int, IsSquareTiles:bool=true) {is=title K2Node_FunctionEntry}
   n2 = BreakVector2D(InVec=ColumnSizeToTileSize!(
     self:WBP Discover Tile Base Object Reference=Get WBP_Discover_Tile_Base() -> (WBP_Discover_Tile_Base:WBP Discover Tile Base Object Reference)
     NewColumnSize:int=@n1
     IsSquareTiles:bool=true
    ) -> (NewTileSize:Vector 2D Structure))
   n3 = IsUsingCompactUI()
   n4 = float + float(A=@n2.Y, B=95)
   n5 = "Max (Float)"(A=@n4, B=600)
   > Return Node!(execute, TileSize:Vector 2D Structure=MakeVector2D(X=@n2, Y=SelectFloat(A=@n5, B=@n4, bPickA=@n3)))
  ]}
 CF_ReactionCountToVisibility {nodes [
   n1 = CF_ReactionCountToVisibility!() -> (then, Count:int) {is=title K2Node_FunctionEntry}
   > Return Node!(
    execute
    Visibility:ESlateVisibility Enum=Select(Option 0:ESlateVisibility Enum=Collapsed, Option 1:ESlateVisibility Enum=SelfHitTestInvisible, Index="integer > integer"(A=@n1)) -> (ReturnValue:ESlateVisibility Enum)
   )
  ]}
 CF_TextEndedVisibility {nodes [
   n1 = CF_TextEndedVisibility!() -> (then, EventStatus:ECommunityEventStatus Enum=Unknown) {is=title K2Node_FunctionEntry}
   > Return Node(Visibility=Select!(
     Unknown:ESlateVisibility Enum=Collapsed
     Upcoming:ESlateVisibility Enum=Collapsed
     Live:ESlateVisibility Enum=Collapsed
     Ended:ESlateVisibility Enum=SelfHitTestInvisible
     Canceled:ESlateVisibility Enum=Collapsed
     Index:ECommunityEventStatus Enum=@n1
    ) -> (ReturnValue))
  ]}
 CF_TextLiveVisibility {nodes [
   n1 = CF_TextLiveVisibility!() -> (then, EventStatus:ECommunityEventStatus Enum=Unknown) {is=title K2Node_FunctionEntry}
   > Return Node(Visibility=Select(Unknown=Collapsed, Upcoming=Collapsed, Live=SelfHitTestInvisible, Ended=Collapsed, Canceled=Collapsed, Index=@n1))
  ]}
 CF_TextUpcomingVisibility {nodes [
   n1 = CF_TextUpcomingVisibility!() -> (then, EventStatus:ECommunityEventStatus Enum=Unknown) {is=title K2Node_FunctionEntry}
   > Return Node(Visibility=Select(Unknown=Collapsed, Upcoming=SelfHitTestInvisible, Live=Collapsed, Ended=SelfHitTestInvisible, Canceled=SelfHitTestInvisible, Index=@n1))
  ]}
 Display_DesignTimePreview {nodes [
   Display_DesignTimePreview() {is=title K2Node_FunctionEntry}
   > Set Tile Size!(execute, self:WBP Discover Tile Base Object Reference=Get WBP_Discover_Tile_Base(), InSize:Vector 2D Structure=(X=396,Y=491)) -> (then) {is=title K2Node_CallFunction}
   n1 = Get WBP_Discover_Tile_Base()
   > SetVisibility(
    self=[
     Get Slot_Right!(self:WBP Discover Tile Base Object Reference=@n1) -> (Slot_Right:Named Slot Object Reference)
     Get Slot_OverlayContent!(self:WBP Discover Tile Base Object Reference=@n1) -> (Slot_OverlayContent:Named Slot Object Reference)
    ]
    InVisibility=Collapsed
   )
   CF_ParamsToTileSize!(self, ColumnSize:int=3, IsSquareTiles:bool=true) -> (TileSize:Vector 2D Structure) {is=title K2Node_CallFunction}
  ]}
 EventGraph {nodes [
   "On Loading Errored (Image_Thumb)"!() -> (OutputDelegate, then, ErrorCode:int=0, ModuleId:int=0, ErrorString:String, ErrorMessage:Text) {is=title K2Node_ComponentBoundEvent}
   > SetThumbnailImageFallback()
   n1 = Event OnListItemObjectSet()
   > n2 = Cast To FortCommunityHighlightTileVM!(execute, Object=@n1.ListItemObject) -> (then, CastFailed, AsCommunity Highlight Tile View Model:Community Highlight Tile View Model Object Reference) {is=title K2Node_DynamicCast}
   > n3 = Set FortCommunityHighlightTileVM(FortCommunityHighlightTileVM:Community Highlight Tile View Model Object Reference=@n2) -> (Output_Get:Community Highlight Tile View Model Object Reference)
   > SendImpression!(
    execute
    self:Community Post VM Object Reference=Get Highlight!(self:Community Highlight Tile View Model Object Reference=@n3) -> (Highlight:Community Highlight VM Object Reference)
    LocationContext:String=Tile
    bInteraction:bool=false
   ) -> (then)
   Event On Clicked()
   > SendImpression(
    self=Get Highlight(self=Get FortCommunityHighlightTileVM() -> (FortCommunityHighlightTileVM:Community Highlight Tile View Model Object Reference))
    LocationContext=Full
    bInteraction=true
   )
   > n4 = Cast To FortPlayerController(Object=GetOwningPlayer())
   > n5 = HandleCommunityTileClicked!(
    execute
    self:Fort Activity Selector Object Reference=Get FortActivitySelector() -> (ReturnValue:Fort Activity Selector Object Reference) {is=title K2Node_GetSubsystem}
    FortPC:Fort Player Controller Object Reference=@n4
    InTileItemVM:Community Highlight Tile View Model Object Reference=Get FortCommunityHighlightTileVM()
   ) -> (then)
   n6 = Create Widget!(
    execute:Exec=[@n5.then, @n4.CastFailed]
    Class:User Widget Class Reference=&/WBP_CommunityPost_Modal.WBP_CommunityPost_Modal_C
    OwningPlayer:Player Controller Object Reference
   ) -> (then:Exec, ReturnValue:WBP Community Post Modal Object Reference)
   n7 = Get Highlight(self=Get FortCommunityHighlightTileVM())
   > Set with Broadcast CommunityHighlightId!(
    execute
    CommunityHighlightId:String=GetHighlightId!(self:Community Highlight VM Object Reference=@n7) -> (ReturnValue:String)
    self:WBP Community Post Modal Object Reference=@n6
   ) -> (then, Output_Get:String)
   n8 = GetRuntimeOptions()
   > Set with Broadcast IsEvent!(
    execute
    IsEvent:bool=AND Boolean(
     A=Get bEnableCreatorEvents!(self:Fort Runtime Options Object Reference=@n8) -> (bEnableCreatorEvents:bool)
     B=IsEvent!(self:Community Post VM Object Reference=@n7) -> (ReturnValue:bool)
    )
    self:WBP Community Post Modal Object Reference=@n6
   ) -> (then, Output_Get:bool)
   > TryPushWidgetAsModalPanel!(execute, PushWidget:User Widget Object Reference=@n6) -> (then)
   Event Construct()
   > Parent: Construct()
   > Bind Event to On Context Menu Visibility Changed(self:WBP Discover Tile Base Object Reference=Get WBP_Discover_Tile_Base(), Delegate=Create Event()) {is=title K2Node_AddDelegate}
   > UpdateContextMenuIconVisibility()
   n9 = Event PreConstruct()
   > SetMobileStyle() {is=title K2Node_CallFunction}
   > Branch(Condition=@n9.IsDesignTime)
   > Display_DesignTimePreview() {is=title K2Node_CallFunction}
   Event Destruct()
   > Parent: Destruct()
   > Unbind all Events from On Context Menu Visibility Changed!(execute:Exec, self:WBP Discover Tile Base Object Reference=Get WBP_Discover_Tile_Base()) -> (then:Exec) {is=title K2Node_ClearDelegate}
   Event On Input Method Changed!() -> (OutputDelegate, then, CurrentInputType:ECommonInputType Enum) {is=title K2Node_Event}
   > UpdateContextMenuIconVisibility()
   n10 = Event SetNestedButtonGroupSelectionVM!() -> (OutputDelegate, then, InVM:Fort Nested Button Group Selection VM Object Reference) {is=title K2Node_Event}
   > Set FortNestedButtonGroupSelectionVM(FortNestedButtonGroupSelectionVM:Fort Nested Button Group Selection VM Object Reference=@n10.InVM) -> (Output_Get:Fort Nested Button Group Selection VM Object Reference)
  ]}
 GetNestedButtonGroupSelectionVM {nodes [
   GetNestedButtonGroupSelectionVM() {is=title K2Node_FunctionEntry}
   > Return Node!(
    execute
    ReturnValue:Fort Nested Button Group Selection VM Object Reference=Get FortNestedButtonGroupSelectionVM() -> (FortNestedButtonGroupSelectionVM:Fort Nested Button Group Selection VM Object Reference)
   )
  ]}
 OnContextMenuActionClicked {nodes [
   n1 = OnContextMenuActionClicked!() -> (then, InteractionNameID:Name) {is=title K2Node_FunctionEntry}
   > n2 = Sequence()
   > n3 = Branch(Condition="Equal (Name)"(A=@n1, B=Get OpenInBrowser_ActionId() -> (OpenInBrowser_ActionId:Name=None)))
   > ViewReplies(self:Community Post VM Object Reference=Get Highlight(self=Get FortCommunityHighlightTileVM()))
   n4 = Branch(execute=@n3.else, Condition="Equal (Name)"(A=@n1, B=Get ViewCreator_ActionId() -> (ViewCreator_ActionId:Name=None)))
   > ViewCreatorPage(self:Community Highlight Tile View Model Object Reference=Get FortCommunityHighlightTileVM())
   Branch(execute=@n4.else, Condition="Equal (Name)"(A=@n1, B=Get ViewIslandDetails_ActionId() -> (ViewIslandDetails_ActionId:Name=None)))
   > ViewIslandDetails(self:Community Highlight Tile View Model Object Reference=Get FortCommunityHighlightTileVM())
   CloseContextMenu(execute=@n2.then_1, self:WBP Discover Tile Base Object Reference=Get WBP_Discover_Tile_Base())
  ]}
 OnContextMenuVisibilityChanged {nodes [
   n1 = OnContextMenuVisibilityChanged!() -> (then, IsVisible:bool=~) {is=title K2Node_FunctionEntry}
   n2 = SlotAsVerticalBoxSlot(Widget=Get CreatorLine() -> (CreatorLine:WBP Community Creator Line Object Reference))
   > n3 = Set members in Margin!(
    execute:Exec
    StructRef:"Margin Structure (by ref)"=Get Padding!(self:Vertical Box Slot Object Reference=@n2) -> (Padding:Margin Structure)
    Right:float=SelectFloat(A=56, B=16, bPickA=@n1)
   ) -> (then:Exec, StructOut:"Margin Structure (by ref)") {is=title K2Node_SetFieldsInStruct}
   > SetPadding!(execute, self:Vertical Box Slot Object Reference=@n2, InPadding:Margin Structure=@n3) -> (then)
  ]}
 SetMobileStyle {nodes [
   SetMobileStyle() {is=title K2Node_FunctionEntry}
   > Branch(Condition=IsUsingCompactUI())
   > SetRowFill!(execute, self:Grid Panel Object Reference=Get GP_Main() -> (GP_Main:Grid Panel Object Reference), RowIndex:int=0, Coefficient:float=0.7) -> (then)
   > n1 = GetDynamicMaterial(self=Get Backplate!(self:WBP UIKit Backplate Base Object Reference=Get WBP_UIKit_Backplate() -> (WBP_UIKit_Backplate:WBP UIKit Backplate Object Reference)) -> (Backplate:Image Object Reference))
   > SetScalarParameterValue(self=@n1, ParameterName=Radius in Pixels, Value=1)
   > SetVectorParameterValue(
    self=@n1
    ParameterName=Corner Radius
    Value=Make Linear Color!(R:float=10, G:float=10, B:float=10, A:float=10) -> (LinearColor:Linear Color Structure) {is=title K2Node_MakeStruct}
   )
  ]}
 SetThumbnailImageFallback {nodes [
   SetThumbnailImageFallback() {is=title K2Node_FunctionEntry}
   > n1 = GetDynamicMaterial(self=Get Image_Thumb() -> (Image_Thumb:Fort Lazy Image Object Reference))
   > SetTextureParameterValue(self=@n1, ParameterName=ThumbnailImage, Value=/FortniteCoreUI/UI/Community/Art/Textures/T_UI_CommunityPost_Banner_FallbackError)
  ]}
 Set_ImageURL {nodes [
   n1 = Set_ImageURL!() -> (then, URL:String) {is=title K2Node_FunctionEntry}
   > SetBrushFromURL!(
    execute
    self:Fort Lazy Image Object Reference=Get Image_Thumb()
    CMSImageURL:String=@n1
    bMatchSize:bool=false
    Delay:float=0
    DownloadPriority:ECMSDownloadPriority Enum=Normal
   ) -> (then)
  ]}
 Set_OpenInBrowser_Action {nodes [
   n1 = Set_OpenInBrowser_Action!() -> (then, IsActionShown:bool=~) {is=title K2Node_FunctionEntry}
   > UpdateContextMenuAction!(
    execute
    self
    ActionDisplayName:Text="LOCTEXT(\"Open in Browser\")"
    ActionId:Name=Get OpenInBrowser_ActionId() -> (OpenInBrowser_ActionId=None)
    IsActionShown:bool=@n1
    ActionSortPriority:int=500
   ) -> (then)
  ]}
 Set_ViewCreator_Action {nodes [
   n1 = Set_ViewCreator_Action!() -> (then, IsActionShown:bool=~) {is=title K2Node_FunctionEntry}
   > UpdateContextMenuAction(
    ActionDisplayName="LOCTEXT(\"View Creator\")"
    ActionId=Get ViewCreator_ActionId() -> (ViewCreator_ActionId=None)
    IsActionShown=@n1
    ActionSortPriority=1000
   )
  ]}
 Set_ViewIslandDetails_Action {nodes [
   n1 = Set_ViewIslandDetails_Action!() -> (then, IsActionShown:bool=~) {is=title K2Node_FunctionEntry}
   > UpdateContextMenuAction(
    ActionDisplayName="LOCTEXT(\"View Island Details\")"
    ActionId=Get ViewIslandDetails_ActionId() -> (ViewIslandDetails_ActionId=None)
    IsActionShown=@n1
    ActionSortPriority=1001
   )
  ]}
 UpdateContextMenuAction {nodes [
   UpdateContextMenuAction!() -> (then, ActionDisplayName:Text, ActionId:Name, IsActionShown:bool=~, ActionSortPriority:int=1000) {is=title K2Node_FunctionEntry}
   n1 = Get ContextMenuActionContext!(self:WBP Discover Tile Base Object Reference=Get WBP_Discover_Tile_Base()) -> (ContextMenuActionContext:UIKit Context Menu Action Context Object Reference)
   > RemoveInteraction!(execute, self:UIKit Context Menu Action Context Object Reference=@n1, InteractionNameID:Name=Get ActionId() -> (ActionId:Name=None)) -> (then)
   > Branch(Condition=Get IsActionShown() -> (IsActionShown:bool))
   > AddInteraction!(
    execute
    self:UIKit Context Menu Action Context Object Reference=@n1
    InActionInfo_DisplayName:Text=Get ActionDisplayName() -> (ActionDisplayName:Text)
    InActionInfo_IconTexture:Texture 2D Soft Object Reference
    InActionInfo_IconMaterial:Material Instance Soft Object Reference
    InActionInfo_SortPriority:int=Get ActionSortPriority() -> (ActionSortPriority:int=0)
    InActionInfo_bIsAvailable:bool=True
    InActionInfo_bIsEnabled:bool=True
    InActionInfo_bIsSelected:bool=False
    InActionInfo_bAddDivider:bool=False
    InActionInfo_bPositionDividerAfter:bool=False
    InActionInfo_bCloseMenuOnSelect:bool=False
    InActionInfo_InteractionNameID:Name=Get ActionId() -> (ActionId=None)
    InActionInfo_Interaction:Data Table Row Handle Structure=(DataTable=None,RowName="")
    InActionInfo_PressedDelegate:Delegate="(null).None"
    InActionInfo_ReleasedDelegate:Delegate="(null).None"
    InInteractionDelegate:"Delegate (by ref)"=Create Event()
   ) -> (then)
  ]}
 UpdateContextMenuIconVisibility {nodes [
   UpdateContextMenuIconVisibility() {is=title K2Node_FunctionEntry}
   > SetContextMenuIconAlwaysVisible!(
    execute
    self:WBP Discover Tile Base Object Reference=Get WBP_Discover_Tile_Base()
    IsContextMenuIconAlwaysVisible:bool=IsUsingInputType!(InputType:ECommonInputType Enum=Touch, PlayerController:Player Controller Object Reference=GetOwningPlayer()) -> (ReturnValue:bool)
   ) -> (then)
  ]}
 UpdateTileDescriptionVisibility {nodes [
   n1 = UpdateTileDescriptionVisibility!() -> (then, IsEvent:bool=~) {is=title K2Node_FunctionEntry}
   n2 = GetRuntimeOptions()
   > n3 = Branch(Condition=NOT Boolean(A=Get bEnableCreatorEvents(self=@n2)))
   > SetVisibility(self=Get VB_HighlightDescription() -> (VB_HighlightDescription:Vertical Box Object Reference), InVisibility=SelfHitTestInvisible)
   > SetVisibility(self=Get VB_EventDescription() -> (VB_EventDescription:Vertical Box Object Reference), InVisibility=Collapsed)
   SetVisibility(
    execute=@n3.else
    self=Get VB_HighlightDescription()
    InVisibility="To Visibility (Boolean)"(bIsVisible=@n1, TrueVisibility=Collapsed, FalseVisibility=SelfHitTestInvisible) -> (ReturnValue=Visible)
   )
   > SetVisibility(
    self=Get VB_EventDescription()
    InVisibility="To Visibility (Boolean)"(bIsVisible=@n1, TrueVisibility=SelfHitTestInvisible, FalseVisibility=Collapsed) -> (ReturnValue=Visible)
   )
  ]}
]
variables [
 OpenInBrowser_ActionId {type=string, default=OpenInBrowser}
 ViewCreator_ActionId {type=string, default=ViewCreator}
 ViewIslandDetails_ActionId {type=string, default=ViewIsland}
]
widgets {
 info {
  inheritedWidgetCount=0
  namedSlotCount=1
  parentClass=/DiscoveryBrowser/Tiles/WBP_Discover_Tile_Button.WBP_Discover_Tile_Button_C
  rootWidgetClass=null
  widgetCount=30
 }
 widgets [
  SizeBox_0 {
   widgetClassPath=&SizeBox
   widget=SizeBox_0
   properties {
    bOverride_HeightOverride=true
    bOverride_WidthOverride=true
    heightOverride=400
    maxDesiredHeight=400
    maxDesiredWidth=414
    minDesiredHeight=400
    minDesiredWidth=414
    widthOverride=400
   }
  }
  GP_Main {
   widgetClassPath=&GridPanel
   bIsVariable=true
   parent=SizeBox_0
   widget=GP_Main
   properties {columnFill [1], rowFill [0.48,0.52]}
   slot {class=SizeBoxSlot}
  }
  MinSize {widgetClassPath=&Spacer, parent=GP_Main, widget=MinSize, properties {size=<414 400>}, slot {class=GridSlot}}
  WBP_UIKit_Backplate {
   widgetClassPath=&WBP_UIKit_Backplate
   bIsVariable=true
   parent=GP_Main
   widget=WBP_UIKit_Backplate
   properties {brightness=Medium, colorAndOpacity=<rgba 1 1 1 0.502>}
   slot {class=GridSlot, properties {layer=-50, rowSpan=2}}
  }
  Highlight {
   widgetClassPath=/FortUILibrary/UIKit/Highlight/WBP_UIKit_Highlight
   bIsVariable=true
   parent=GP_Main
   widget=Highlight
   properties {highlightText=Live, highlightType=LiveAlt}
   slot {class=GridSlot, properties {horizontalAlignment=HAlign_Right, layer=10, verticalAlignment=VAlign_Top, padding=<ltrb 0 10 10 0>}}
  }
  Image_Thumb {
   widgetClassPath=&FortLazyImage
   bIsVariable=true
   parent=GP_Main
   widget=Image_Thumb
   properties {
    materialTextureParamName=ThumbnailImage
    brush {
     bIsDynamicallyLoaded=false
     drawAs=Image
     imageType=FullColor
     mirroring=NoMirror
     resourceName=None
     resourceObject=/DiscoveryBrowser/Tiles/Community/Materials/MI_UI_CommunityThumbnail
     tiling=NoTile
     imageSize=<32 32>
     margin=<ltrb 0 0 0 0>
     outlineSettings {
      bUseBrushTransparency=false
      roundingType=HalfHeightRadius
      width=0
      color {colorUseRule=UseColor_Specified, specifiedColor=<rgba 0 0 0 0>}
      cornerRadii=<0 0 0 0>
     }
     tintColor {colorUseRule=UseColor_Specified, specifiedColor=<rgba 1 1 1 1>}
     uVRegion {bIsValid=false, max=<0 0>, min=<0 0>}
    }
    loadingBackgroundBrush {
     bIsDynamicallyLoaded=false
     drawAs=Image
     imageType=FullColor
     mirroring=NoMirror
     resourceName=None
     resourceObject=/FortniteCoreUI/UI/Community/Art/M_UI_CommunityPost_Loading
     tiling=NoTile
     imageSize=<32 32>
     margin=<ltrb 0 0 0 0>
     outlineSettings {
      bUseBrushTransparency=false
      roundingType=HalfHeightRadius
      width=0
      color {colorUseRule=UseColor_Specified, specifiedColor=<rgba 0 0 0 0>}
      cornerRadii=<0 0 0 0>
     }
     tintColor {colorUseRule=UseColor_Specified, specifiedColor=<rgba 1 1 1 1>}
     uVRegion {bIsValid=false, max=<0 0>, min=<0 0>}
    }
    loadingThrobberBrush {
     bIsDynamicallyLoaded=false
     drawAs=Image
     imageType=FullColor
     mirroring=NoMirror
     resourceName=None
     resourceObject=/Game/Global/Textures/Default/Blanks/T_Blank_White_ZeroAlpha_sRGBColor
     tiling=NoTile
     imageSize=<1 1>
     margin=<ltrb 0 0 0 0>
     outlineSettings {
      bUseBrushTransparency=false
      roundingType=HalfHeightRadius
      width=0
      color {colorUseRule=UseColor_Specified, specifiedColor=<rgba 0 0 0 0>}
      cornerRadii=<0 0 0 0>
     }
     tintColor {colorUseRule=UseColor_Specified, specifiedColor=<rgba 1 1 1 1>}
     uVRegion {bIsValid=false, max=<0 0>, min=<0 0>}
    }
   }
   slot {class=GridSlot, properties {rowSpan=1}}
  }
  GridPanel_0 {
   widgetClassPath=&GridPanel
   parent=GP_Main
   widget=GridPanel_0
   properties {columnFill [1], rowFill [1]}
   slot {class=GridSlot, properties {layer=10, row=1}}
  }
  VB_HighlightDescription {
   widgetClassPath=&VerticalBox
   bIsVariable=true
   parent=GridPanel_0
   widget=VB_HighlightDescription
   properties {bOverride_Cursor=true}
   slot {class=GridSlot, properties {padding=<ltrb 16 16 16 13>}}
  }
  SB_TextTitle {
   widgetClassPath=&SizeBox
   bIsVariable=true
   parent=VB_HighlightDescription
   widget=SB_TextTitle
   properties {bOverride_MaxDesiredHeight=true, bOverride_WidthOverride=true, maxDesiredHeight=200, widthOverride=504}
   slot {class=VerticalBoxSlot, properties {size {sizeRule=Fill, value=1}}}
  }
  Text_Title {
   widgetClassPath=&CommonTextBlock
   bIsVariable=true
   parent=SB_TextTitle
   widget=Text_Title
   properties {
    autoWrapText=true
    bIsScrollingEnabled=false
    clipping=ClipToBounds
    mobileFontSizeMultiplier=1.15
    text=Post Title should be really long wwwwwwwwwwwwwwwwwwwwwwwwww
    textOverflowPolicy=MultilineEllipsis
    font {
     bForceMonospaced=false
     bMaterialIsStencil=false
     fontMaterial=null
     fontObject=&HeadingNow
     letterSpacing=0
     monospacedWidth=1
     size=15.75
     skewAmount=0
     typefaceFontName=75 Medium
     outlineSettings {
      bApplyOutlineToDropShadows=false
      bMiteredCorners=false
      bSeparateFillAlpha=false
      outlineMaterial=null
      outlineSize=0
      outlineColor=<rgba 0 0 0 1>
     }
    }
    shadowOffset=<0 0>
   }
   slot {class=SizeBoxSlot}
  }
  HB_Communitytag_TimeStamp {
   widgetClassPath=&HorizontalBox
   parent=VB_HighlightDescription
   widget=HB_Communitytag_TimeStamp
   properties {clipping=ClipToBounds}
   slot {class=VerticalBoxSlot}
  }
  Text_IslandName {
   widgetClassPath=&CommonTextBlock
   bIsVariable=true
   parent=HB_Communitytag_TimeStamp
   widget=Text_IslandName
   properties {
    mobileFontSizeMultiplier=1.15
    text=Island Name
    colorAndOpacity {colorUseRule=UseColor_Specified, specifiedColor=<rgba 0.5776 0.5776 0.5776 1>}
    font {
     bForceMonospaced=false
     bMaterialIsStencil=false
     fontMaterial=null
     fontObject=&HeadingNow
     letterSpacing=0
     monospacedWidth=1
     size=13.5
     skewAmount=0
     typefaceFontName=74 Regular
     outlineSettings {
      bApplyOutlineToDropShadows=false
      bMiteredCorners=false
      bSeparateFillAlpha=false
      outlineMaterial=null
      outlineSize=0
      outlineColor=<rgba 0 0 0 1>
     }
    }
    shadowOffset=<0 0>
   }
   slot {class=HorizontalBoxSlot, properties {horizontalAlignment=HAlign_Left}}
  }
  SizeBox_1 {
   widgetClassPath=&SizeBox
   parent=VB_HighlightDescription
   widget=SizeBox_1
   properties {bOverride_MinDesiredHeight=true, minDesiredHeight=38}
   slot {class=VerticalBoxSlot, properties {verticalAlignment=VAlign_Bottom, padding=<ltrb 0 13 0 0>}}
  }
  HB_Footer {
   widgetClassPath=&HorizontalBox
   bIsVariable=true
   parent=SizeBox_1
   widget=HB_Footer
   properties {bOverride_Cursor=true}
   slot {class=SizeBoxSlot}
  }
  Count_Likes {
   widgetClassPath=/FortniteCoreUI/UI/Community/WBP_CommunityPost_ReactionsLine
   bIsVariable=true
   parent=HB_Footer
   widget=Count_Likes
   properties {
    brush Visibility=SelfHitTestInvisible
    emoji Visibility=Collapsed
    material=/FortniteCoreUI/UI/Community/Art/MaterialInstances/MI_UI_Communities_Reaction_Emoji
    in Color and Opacity=<rgba 1 1 1 1>
   }
   slot {class=HorizontalBoxSlot, properties {padding=<ltrb 0 0 12 0>}}
  }
  Count_Replies {
   widgetClassPath=&WBP_CommunityPost_ReactionsLine
   bIsVariable=true
   parent=HB_Footer
   widget=Count_Replies
   properties {emoji="💬", material=/FortUILibrary/Textures/MaterialInstances/MI_UIKit_Icon_ChannelWhsiper, in Color and Opacity=<rgba 0.5191 0.7428 1 1>}
   slot {class=HorizontalBoxSlot, properties {horizontalAlignment=HAlign_Left, size {sizeRule=Fill, value=1}}}
  }
  VB_EventDescription {
   widgetClassPath=&VerticalBox
   bIsVariable=true
   parent=GridPanel_0
   widget=VB_EventDescription
   properties {visibility=Collapsed}
   slot {class=GridSlot, properties {padding=<ltrb 16 16 16 13>}}
  }
  GridPanel_1 {
   widgetClassPath=&GridPanel
   parent=VB_EventDescription
   widget=GridPanel_1
   properties {columnFill [0,1], rowFill [1]}
   slot {class=VerticalBoxSlot, properties {padding=<ltrb 0 0 0 13>, size {sizeRule=Fill, value=1}}}
  }
  StartDate {
   widgetClassPath=/FortniteCoreUI/UI/Community/WBP_Community_Date
   bIsVariable=true
   parent=GridPanel_1
   widget=StartDate
   properties {bOverride_Cursor=true}
   slot {class=GridSlot, properties {horizontalAlignment=HAlign_Left}}
  }
  VerticalBox_0 {
   widgetClassPath=&VerticalBox
   parent=GridPanel_1
   widget=VerticalBox_0
   slot {class=GridSlot, properties {column=1, padding=<ltrb 16 0 0 0>}}
  }
  Text_CommunityEventTimeUpcoming {
   widgetClassPath=&CommonTextBlock
   bIsVariable=true
   parent=VerticalBox_0
   widget=Text_CommunityEventTimeUpcoming
   properties {
    mobileFontSizeMultiplier=1.4
    scrollStyle=&TextScrollStyle-Base_C
    text="{time} - {time}"
    colorAndOpacity {colorUseRule=UseColor_Specified, specifiedColor=<rgba 0.5776 0.5776 0.5776 1>}
    font {
     bForceMonospaced=false
     bMaterialIsStencil=false
     fontMaterial=null
     fontObject=&HeadingNow
     letterSpacing=0
     monospacedWidth=1
     size=11.25
     skewAmount=0
     typefaceFontName=75 Medium
     outlineSettings {
      bApplyOutlineToDropShadows=false
      bMiteredCorners=false
      bSeparateFillAlpha=false
      outlineMaterial=null
      outlineSize=0
      outlineColor=<rgba 0 0 0 1>
     }
    }
    shadowOffset=<0 0>
   }
   slot {class=VerticalBoxSlot, properties {padding=<ltrb 0 4 0 8>}}
  }
  Text_CommunityEventTimeEnded {
   widgetClassPath=&CommonTextBlock
   bIsVariable=true
   parent=VerticalBox_0
   widget=Text_CommunityEventTimeEnded
   properties {
    mobileFontSizeMultiplier=1.4
    scrollStyle=&TextScrollStyle-Base_C
    text="ended {Date} ago "
    colorAndOpacity {colorUseRule=UseColor_Specified, specifiedColor=<rgba 0.5776 0.5776 0.5776 1>}
    font {
     bForceMonospaced=false
     bMaterialIsStencil=false
     fontMaterial=null
     fontObject=&HeadingNow
     letterSpacing=0
     monospacedWidth=1
     size=11.25
     skewAmount=0
     typefaceFontName=75 Medium
     outlineSettings {
      bApplyOutlineToDropShadows=false
      bMiteredCorners=false
      bSeparateFillAlpha=false
      outlineMaterial=null
      outlineSize=0
      outlineColor=<rgba 0 0 0 1>
     }
    }
    shadowOffset=<0 0>
   }
   slot {class=VerticalBoxSlot, properties {padding=<ltrb 0 4 0 8>}}
  }
  Text_CommunityEventTimeLive {
   widgetClassPath=&CommonTextBlock
   bIsVariable=true
   parent=VerticalBox_0
   widget=Text_CommunityEventTimeLive
   properties {
    mobileFontSizeMultiplier=1.4
    scrollStyle=&TextScrollStyle-Base_C
    text="ONGOING • ENDS IN {time} Hours"
    colorAndOpacity {colorUseRule=UseColor_Specified, specifiedColor=<rgba 0 1 0.6038 1>}
    font {
     bForceMonospaced=false
     bMaterialIsStencil=false
     fontMaterial=null
     fontObject=&HeadingNow
     letterSpacing=0
     monospacedWidth=1
     size=11.25
     skewAmount=0
     typefaceFontName=75 Medium
     outlineSettings {
      bApplyOutlineToDropShadows=false
      bMiteredCorners=false
      bSeparateFillAlpha=false
      outlineMaterial=null
      outlineSize=0
      outlineColor=<rgba 0 0 0 1>
     }
    }
    shadowOffset=<0 0>
   }
   slot {class=VerticalBoxSlot, properties {padding=<ltrb 0 4 0 8>}}
  }
  Text_CommunityEventTitle {
   widgetClassPath=&CommonTextBlock
   bIsVariable=true
   parent=VerticalBox_0
   widget=Text_CommunityEventTitle
   properties {
    autoWrapText=true
    bIsScrollingEnabled=false
    clipping=ClipToBounds
    mobileFontSizeMultiplier=1.15
    text=Post Title should be really long wwwwwwwwwwwwwwwwwwwwwwwwww
    textOverflowPolicy=MultilineEllipsis
    font {
     bForceMonospaced=false
     bMaterialIsStencil=false
     fontMaterial=null
     fontObject=&HeadingNow
     letterSpacing=0
     monospacedWidth=1
     size=15.75
     skewAmount=0
     typefaceFontName=75 Medium
     outlineSettings {
      bApplyOutlineToDropShadows=false
      bMiteredCorners=false
      bSeparateFillAlpha=false
      outlineMaterial=null
      outlineSize=0
      outlineColor=<rgba 0 0 0 1>
     }
    }
    shadowOffset=<0 0>
   }
   slot {class=VerticalBoxSlot, properties {size {sizeRule=Fill, value=1}}}
  }
  Text_CommunityEventIslandName {
   widgetClassPath=&CommonTextBlock
   bIsVariable=true
   parent=VerticalBox_0
   widget=Text_CommunityEventIslandName
   properties {
    autoWrapText=true
    mobileFontSizeMultiplier=1.3
    text=Island Name
    colorAndOpacity {colorUseRule=UseColor_Specified, specifiedColor=<rgba 0.5776 0.5776 0.5776 1>}
    font {
     bForceMonospaced=false
     bMaterialIsStencil=false
     fontMaterial=null
     fontObject=&HeadingNow
     letterSpacing=0
     monospacedWidth=1
     size=12.75
     skewAmount=0
     typefaceFontName=74 Regular
     outlineSettings {
      bApplyOutlineToDropShadows=false
      bMiteredCorners=false
      bSeparateFillAlpha=false
      outlineMaterial=null
      outlineSize=0
      outlineColor=<rgba 0 0 0 1>
     }
    }
    shadowOffset=<0 0>
   }
   slot {class=VerticalBoxSlot, properties {padding=<ltrb 0 8 0 0>}}
  }
  SizeBox {
   widgetClassPath=&SizeBox
   parent=VB_EventDescription
   widget=SizeBox
   properties {bOverride_MinDesiredHeight=true, minDesiredHeight=38}
   slot {class=VerticalBoxSlot, properties {verticalAlignment=VAlign_Bottom}}
  }
  InterestedLine {
   widgetClassPath=&/WBP_CommunityEvent_InterestedLine
   bIsVariable=true
   parent=SizeBox
   widget=InterestedLine
   slot {class=SizeBoxSlot}
  }
  VB_Creator {widgetClassPath=&VerticalBox, parent=GridPanel_0, widget=VB_Creator, slot {class=GridSlot, properties {row=1, rowSpan=1}}}
  Divider {
   widgetClassPath=&Image
   parent=VB_Creator
   widget=Divider
   properties {
    visibility=SelfHitTestInvisible
    brush {
     bIsDynamicallyLoaded=false
     drawAs=Image
     imageType=NoImage
     mirroring=NoMirror
     resourceName=None
     resourceObject=null
     tiling=NoTile
     imageSize=<32 2>
     margin=<ltrb 0 0 0 0>
     outlineSettings {
      bUseBrushTransparency=false
      roundingType=HalfHeightRadius
      width=0
      color {colorUseRule=UseColor_Specified, specifiedColor=<rgba 0 0 0 0>}
      cornerRadii=<0 0 0 0>
     }
     tintColor {colorUseRule=UseColor_Specified, specifiedColor=<rgba 1 1 1 1>}
     uVRegion {bIsValid=false, max=<0 0>, min=<0 0>}
    }
    colorAndOpacity=<rgba 1 1 1 0.1>
   }
   slot {class=VerticalBoxSlot}
  }
  CreatorLine {
   widgetClassPath=&/WBP_Community_CreatorLine
   bIsVariable=true
   parent=VB_Creator
   widget=CreatorLine
   slot {class=VerticalBoxSlot, properties {padding=<ltrb 16 16 16 16>}}
  }
 ]
}
